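# Animation Blueprint for the M_MED_SpongeHollow male player cosmetic body, copying the attached parent mesh's pose and adapting it to this skeleton.
# The copied pose feeds a Control Rig driven by emote and crouch values set in the EventGraph.
# An additive pelvis offset and a chain of bone constraints on arm twist and deform bones follow.
# An LOD 2 flag updates each frame but is unread in the AnimGraph.
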AnimBlueprint /BRCosmetics/Characters/Player/Male/Medium/Bodies/M_MED_SpongeHollow/Meshes/M_MED_SpongeHollow_AnimBP
target_skeleton=&/M_MED_SpongeHollow_Skeleton
graphs [
 AnimGraph {nodes [
   n1 = Control Rig!(
    Source=Copy Pose From Mesh() {bUseAttachedParent=true}
    Alpha
    isPlayingEmote_CR:bool=Get bIsPlayingEmote()
    isNotEmotingAlpha_CR:double=Get Is Not Emoting Alpha() -> (Is Not Emoting Alpha:double)
    isCrouching_CR:double=Get Is Crouching Alpha() -> (Is Crouching Alpha:double)
   ) -> (Pose) {controlRigAssetReference {blueprintRigClass=&/M_MED_SpongeHollow_CtrlRig.M_MED_SpongeHollow_CtrlRig_C}}
   n2 = "Transform (Modify) Bone - Bone: dfrm_pelvis"(ComponentPose=Local To Component(LocalPose=@n1), Alpha=1, Translation="3,0,0", Rotation="0.005,13.278,-0.009", Scale="1,1,1") {
    rotationMode=BMM_Additive
    rotationSpace=BCS_BoneSpace
    translationMode=BMM_Additive
    translationSpace=BCS_BoneSpace
    alphaBoolBlend {blendOption=Linear}
    alphaScaleBiasClamp {bInterpResult=true, interpSpeedIncreasing=6}
    boneToModify {boneName=dfrm_pelvis}
    rotation=<rot 0.0051 13.2779 -0.009>
    translation {x=3}
   }
   n3 = "Constraint - upperarm_twist_01_r (P)"!(ComponentPose=@n2, bAlphaBoolEnabled:bool=True, ConstraintWeights_0=1) -> (Pose) {
    alphaInputType=Bool
    alphaBoolBlend {blendOption=Linear}
    boneToModify {boneName=upperarm_twist_01_r}
    constraintSetup [{targetBone {boneName=upperarm_r}}]
    constraintWeights [1]
   }
   n4 = "Constraint - deform_upperarm_bk_base_r (P)"!(ComponentPose=@n3, bAlphaBoolEnabled:bool=True, ConstraintWeights_0=1) -> (Pose) {
    alphaInputType=Bool
    alphaBoolBlend {blendOption=Linear}
    boneToModify {boneName=deform_upperarm_bk_base_r}
    constraintSetup [{targetBone {boneName=upperarm_twist_01_r}}]
    constraintWeights [1]
   }
   n5 = "Constraint - deform_upperarm_bk_r (P)"!(ComponentPose=@n4, bAlphaBoolEnabled:bool=True, ConstraintWeights_0=1) -> (Pose) {
    alphaInputType=Bool
    alphaBoolBlend {blendOption=Linear}
    boneToModify {boneName=deform_upperarm_bk_r}
    constraintSetup [{targetBone {boneName=deform_upperarm_bk_base_r}}]
    constraintWeights [1]
   }
   n6 = "Constraint - deform_upperarm_fr_base_r (P)"!(ComponentPose=@n5, bAlphaBoolEnabled:bool=True, ConstraintWeights_0=1) -> (Pose) {
    alphaInputType=Bool
    alphaBoolBlend {blendOption=Linear}
    boneToModify {boneName=deform_upperarm_fr_base_r}
    constraintSetup [{targetBone {boneName=upperarm_twist_01_r}}]
    constraintWeights [1]
   }
   n7 = "Constraint - deform_upperarm_bk_r (P)"(ComponentPose=@n6, bAlphaBoolEnabled=True, ConstraintWeights_0=1) {
    alphaInputType=Bool
    alphaBoolBlend {blendOption=Linear}
    boneToModify {boneName=deform_upperarm_bk_r}
    constraintSetup [{targetBone {boneName=deform_upperarm_fr_r}}]
    constraintWeights [1]
   }
   n8 = "Constraint - upperarm_twist_02_r (P)"!(ComponentPose=@n7, bAlphaBoolEnabled:bool=True, ConstraintWeights_0=1) -> (Pose) {
    alphaInputType=Bool
    alphaBoolBlend {blendOption=Linear}
    boneToModify {boneName=upperarm_twist_02_r}
    constraintSetup [{targetBone {boneName=upperarm_r}}]
    constraintWeights [1]
   }
   n9 = "Constraint - deform_upperarm_up_base_r (P)"!(ComponentPose=@n8, bAlphaBoolEnabled:bool=True, ConstraintWeights_0=1) -> (Pose) {
    alphaInputType=Bool
    alphaBoolBlend {blendOption=Linear}
    boneToModify {boneName=deform_upperarm_up_base_r}
    constraintSetup [{targetBone {boneName=upperarm_twist_01_r}}]
    constraintWeights [1]
   }
   n10 = "Constraint - deform_upperarm_up_r (P)"!(ComponentPose=@n9, bAlphaBoolEnabled:bool=True, ConstraintWeights_0=1) -> (Pose) {
    alphaInputType=Bool
    alphaBoolBlend {blendOption=Linear}
    boneToModify {boneName=deform_upperarm_up_r}
    constraintSetup [{targetBone {boneName=deform_upperarm_up_base_r}}]
    constraintWeights [1]
   }
   n11 = "Constraint - upperarm_twist_01_l (P)"!(ComponentPose=@n10, bAlphaBoolEnabled:bool=True, ConstraintWeights_0=1) -> (Pose) {
    alphaInputType=Bool
    alphaBoolBlend {blendOption=Linear}
    boneToModify {boneName=upperarm_twist_01_l}
    constraintSetup [{targetBone {boneName=upperarm_l}}]
    constraintWeights [1]
   }
   n12 = "Constraint - deform_upperarm_bk_base_l (P)"!(ComponentPose=@n11, bAlphaBoolEnabled:bool=True, ConstraintWeights_0=1) -> (Pose) {
    alphaInputType=Bool
    alphaBoolBlend {blendOption=Linear}
    boneToModify {boneName=deform_upperarm_bk_base_l}
    constraintSetup [{targetBone {boneName=upperarm_l}}]
    constraintWeights [1]
   }
   n13 = "Constraint - deform_upperarm_bk_l (P)"!(ComponentPose=@n12, bAlphaBoolEnabled:bool=True, ConstraintWeights_0=1) -> (Pose) {
    alphaInputType=Bool
    alphaBoolBlend {blendOption=Linear}
    boneToModify {boneName=deform_upperarm_bk_l}
    constraintSetup [{targetBone {boneName=deform_upperarm_bk_l}}]
    constraintWeights [1]
   }
   n14 = "Constraint - deform_upperarm_fr_base_l (P)"!(ComponentPose=@n13, bAlphaBoolEnabled:bool=True, ConstraintWeights_0=1) -> (Pose) {
    alphaInputType=Bool
    alphaBoolBlend {blendOption=Linear}
    boneToModify {boneName=deform_upperarm_fr_base_l}
    constraintSetup [{targetBone {boneName=upperarm_l}}]
    constraintWeights [1]
   }
   n15 = "Constraint - deform_upperarm_fr_l (P)"!(ComponentPose=@n14, bAlphaBoolEnabled:bool=True, ConstraintWeights_0=1) -> (Pose) {
    alphaInputType=Bool
    alphaBoolBlend {blendOption=Linear}
    boneToModify {boneName=deform_upperarm_fr_l}
    constraintSetup [{targetBone {boneName=deform_upperarm_fr_base_l}}]
    constraintWeights [1]
   }
   n16 = "Constraint - deform_upperarm_up_base_l (P)"!(ComponentPose=@n15, bAlphaBoolEnabled:bool=True, ConstraintWeights_0=1) -> (Pose) {
    alphaInputType=Bool
    alphaBoolBlend {blendOption=Linear}
    boneToModify {boneName=deform_upperarm_up_base_l}
    constraintSetup [{targetBone {boneName=upperarm_l}}]
    constraintWeights [1]
   }
   n17 = "Constraint - deform_upperarm_up_l (P)"!(ComponentPose=@n16, bAlphaBoolEnabled:bool=True, ConstraintWeights_0=1) -> (Pose) {
    alphaInputType=Bool
    alphaBoolBlend {blendOption=Linear}
    boneToModify {boneName=deform_upperarm_up_l}
    constraintSetup [{targetBone {boneName=deform_upperarm_up_base_l}}]
    constraintWeights [1]
   }
   n18 = "Constraint - upperarm_twist_02_l (P)"!(ComponentPose=@n17, bAlphaBoolEnabled:bool=True, ConstraintWeights_0=1) -> (Pose) {
    alphaInputType=Bool
    alphaBoolBlend {blendOption=Linear}
    boneToModify {boneName=upperarm_twist_02_l}
    constraintSetup [{targetBone {boneName=upperarm_l}}]
    constraintWeights [1]
   }
   n19 = "Constraint - lowerarm_twist_01_l (P)"!(ComponentPose=@n18, bAlphaBoolEnabled:bool=True, ConstraintWeights_0=1) -> (Pose) {
    alphaInputType=Bool
    alphaBoolBlend {blendOption=Linear}
    boneToModify {boneName=lowerarm_twist_01_l}
    constraintSetup [{targetBone {boneName=lowerarm_l}}]
    constraintWeights [1]
   }
   n20 = "Constraint - lowerarm_twist_02_l (P)"!(ComponentPose=@n19, bAlphaBoolEnabled:bool=True, ConstraintWeights_0=1) -> (Pose) {
    alphaInputType=Bool
    alphaBoolBlend {blendOption=Linear}
    boneToModify {boneName=lowerarm_twist_02_l}
    constraintSetup [{targetBone {boneName=lowerarm_l}}]
    constraintWeights [1]
   }
   n21 = "Constraint - lowerarm_twist_02_r (P)"!(ComponentPose=@n20, bAlphaBoolEnabled:bool=True, ConstraintWeights_0=1) -> (Pose) {
    alphaInputType=Bool
    alphaBoolBlend {blendOption=Linear}
    boneToModify {boneName=lowerarm_twist_02_r}
    constraintSetup [{targetBone {boneName=lowerarm_r}}]
    constraintWeights [1]
   }
   n22 = "Constraint - lowerarm_twist_01_r (P)"!(ComponentPose=@n21, bAlphaBoolEnabled:bool=True, ConstraintWeights_0=1) -> (Pose) {
    alphaInputType=Bool
    alphaBoolBlend {blendOption=Linear}
    boneToModify {boneName=lowerarm_twist_01_r}
    constraintSetup [{targetBone {boneName=lowerarm_r}}]
    constraintWeights [1]
   }
   Output Pose(Result=Component To Local(ComponentPose=@n22)) {layerGroup=None}
  ]}
 EventGraph {nodes [
   Event BlueprintUpdateAnimation()
   > Set LOD 2(LOD 2:bool="integer >= integer"(A=Get CurrentLOD(), B=2)) -> (Output_Get:bool)
   n1 = Get bIsPlayingEmote()
   > Set Is Not Emoting Alpha(Is Not Emoting Alpha:double="To Float (Boolean)"(InBool=NOT Boolean(A=@n1))) -> (Output_Get:double)
   Event OnBeginCrouching()
   > Set Is Crouching Alpha(Is Crouching Alpha:double=1) -> (Output_Get:double)
   Event OnEndCrouching()
   > Set Is Crouching Alpha(Is Crouching Alpha=0)
  ]}
]
variables [
 Is Crouching Alpha {type=number}
 Is Not Emoting Alpha {type=number}
 LOD 2 {type=boolean}
]
